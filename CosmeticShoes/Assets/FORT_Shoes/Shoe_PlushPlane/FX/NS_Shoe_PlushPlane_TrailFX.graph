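# Niagara trail effect for the PlushPlane cosmetic shoe, emitting short-lived bright cyan-blue mesh particles from both feet.
# A single emitter spawns per distance travelled while a system running-effect flag is set, with running-check, skydiving and hide-FX inputs available.
# Partitioning assigns particles to left or right skeletal bones or sockets, selecting offset and shoe scale per side.
# Particles inherit slight velocity, then move under drag, gravity and curl noise while popping in and shrinking out.
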NiagaraSystem /CosmeticShoes/Assets/FORT_Shoes/Shoe_PlushPlane/FX/NS_Shoe_PlushPlane_TrailFX
dependencies {
 usedDataInterfaces [&NiagaraDataInterfaceCurve, &NiagaraDataInterfaceSkeletalMesh]
 usedDynamicInputs [
  &Boolean_AndOperation
  /Game/Effects/Niagara/DynamicInputs/Logic/FN_IsRunningCheck
  &Boolean_NotOperation
  &Boolean_OrOperation
  &Divide_Float
  &MakeFloatFromVector
 ]
 usedModules [
  &self:SystemSpawnScript.SystemScriptSource.SystemScriptGraph.NiagaraNodeAssignment_0.SetVariables_#1
  &SystemState
  &self:SystemSpawnScript.SystemScriptSource.SystemScriptGraph.NiagaraNodeAssignment_1.SetVariables_#2
  &self:SystemSpawnScript.SystemScriptSource.SystemScriptGraph.NiagaraNodeAssignment_2.SetVariables_#3
  &self:SystemSpawnScript.SystemScriptSource.SystemScriptGraph.NiagaraNodeAssignment_3.SetVariables_#4
 ]
}
emitters {ShoeSpawnPerUnit {
  emitterUpdateScript [
   &EmitterState()
   &SpawnPerUnit(
    Module.Use Max Movement Threshold=true
    Module.Use Movement Tolerance=false
    Spawn Spacing:NiagaraFloat=&MakeCustomFloatFromBool(Bool=$System.DisplayRunningEffect, True Float=15, False Float=0)
    Use Spawn Probability=false
   )
  ]
  particleSpawnScript [
   &PartitionParticles()
   &InitializeParticle(
    Color=&RandomRangeLinearColor(Minimum=<rgba 0 11.873 15 1>, Maximum=<rgba 0 8.25 15 1>)
    Lifetime=&RandomRangeFloat(Minimum=0.25, Maximum=0.35)
    Mesh Scale=&Multiply_VectorByFloat(
     Vector=&VectorFromFloat(Value=&RandomRangeFloat(Minimum=0.675, Maximum=0.8))
     Float=&MaskFloat(
      Mask Source=1
      Mask Parameter=$Particles.PartitionParticles.Partition
      Mask=0
      Masked Float=$System.ScalePercentL
      Passthrough Float=$System.ScalePercentR
     )
    )
    Mesh Scale Mode=::NewEnumerator4
    Sprite Size Mode=::NewEnumerator3
    Uniform Sprite Size=&Multiply_Float(
     A=4
     B=&MaskFloat(
      Mask Source=1
      Mask Parameter=$Particles.PartitionParticles.Partition
      Mask=0
      Masked Float=$System.ScalePercentL
      Passthrough Float=$System.ScalePercentR
     )
    )
    UsePositionOffset=false
    Write Material Random=true
   )
   &SkeletalMeshLocation(
    Bone / Socket Index=$Particles.PartitionParticles.Partition
    Bone Sampling Mode=::NewEnumerator4
    MaskSpawnGroup=false
    Sampled Position Offset=&MergeHelper_VectorSwitch(0=<-2 0 0.575>, 1=<2 0 -0.575>, 2=?, 3=?, Index=$Particles.PartitionParticles.Partition, Num=::NewEnumerator0)
    Skeletal Mesh=$System.SkeletalMesh
   )
   &ShapeLocation(
    Hemisphere Distribution=<1 0.5>
    MaskSpawnGroup=false
    Override Seed=false
    Rotation Axis=<0 -1 0>
    Rotation Mode=::NewEnumerator1
    Sphere Radius=4
    UseOwnerScale=true
   )
   &InheritVelocity(Inherited Velocity Amount Scale=<0.12 0.12 0.12>, Speed Limit=true, UseSpeedThreshold=false)
   &InitialMeshOrientation(Use Rotation Vector=false)
  ]
  particleUpdateScript [
   &ParticleState()
   &Drag(Drag=&RandomRangeFloat(Minimum=4, Maximum=5), Use Linear Drag=true, Use Rotational Drag=true)
   &GravityForce(Gravity=<0 0 -888>)
   &ScaleMeshSize(Scale Factor=&VectorFromFloat(Value=&FloatFromCurve(FloatCurve=DataInterface() {propertyValues {Curve {keys [
          {arriveTangent=32.4788, interpMode=RCIM_Cubic, leaveTangent=32.4788, leaveTangentWeight=0.0567, tangentMode=RCTM_User, time=0, value=0}
          {arriveTangentWeight=0.2889, interpMode=RCIM_Cubic, leaveTangentWeight=0.051, tangentMode=RCTM_User, time=0.07, value=1.2}
          {arriveTangentWeight=0.011, interpMode=RCIM_Cubic, leaveTangentWeight=0.1471, tangentMode=RCTM_User, time=0.1, value=1}
          {arriveTangentWeight=0.1333, interpMode=RCIM_Cubic, leaveTangentWeight=0.1333, tangentMode=RCTM_User, time=0.75, value=1}
          {arriveTangentWeight=0.1333, interpMode=RCIM_Cubic, tangentMode=RCTM_User, time=1, value=0}
         ]}}})))
   &CurlNoiseForce(
    Bias Noise=false
    Noise Frequency=25
    Noise Strength=&RandomRangeFloat(Minimum=150, Maximum=225)
    Pan Noise=false
    Random Seed:NiagaraInt32=$Particles.PartitionParticles.Partition
   )
   &SolveForcesAndVelocity(Clamp Velocity=false, Limit Acceleration=false)
  ]
  rendererClasses [&NiagaraMeshRendererProperties]
  renderers [{rendererClass=&NiagaraMeshRendererProperties, rendererIndex=0}]
 }}
summary {
 systemName=NS_Shoe_PlushPlane_TrailFX
 emitters [{emitterName=ShoeSpawnPerUnit, rendererClasses [&NiagaraMeshRendererProperties]}]
 userVariables [
  User.HideFootFX {type=NiagaraBool}
  User.IsSkydiving {type=NiagaraBool}
  User.ShoeScale_L {type=Vector3f, description="", defaultValue {struct=Vector3f, value=<1 1 1>}}
  User.ShoeScale_R {type=Vector3f, description="", defaultValue {struct=Vector3f, value=<1 1 1>}}
 ]
}
